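; Blueprint interface for Creative AI objects that supply AI spawner info, centered on Fort Athena AI spawner data component lists.
; It declares functions to async load a list for a callback actor, query it by pawn or as current, and get the current character item definition.
; Other functions give a list with its spawner and actor, request spawns at a transform, and handle post-spawn context.
; Implementers supply all logic.
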(fn AsyncLoadAISpawnerDataComponentList
  (params (CallbackTarget :type Actor%20Object%20Reference)))

(fn GetAISpawnerDataComponentListByPawn
  (params (FortPawn :type Fort%20Pawn%20Object%20Reference%20%28by%20ref%29))
  (return false 0 0))

(fn GetCurrentAISpawnerDataComponentList
  (return false 0 0))

(fn GetCurrentCharacterItemDefinition
  (return 0))

(fn GiveAISpawnerDataComponentList
  (params (AISpawnerDataComponentList :type Fort%20Athena%20AISpawner%20Data%20Component%20List%20Soft%20Object%20Reference) (AthenaAISpawner :type Athena%20AISpawner%20Object%20Reference) (SpawnerActor :type Actor%20Object%20Reference))
  (return))

(fn HandlePostSpawn
  (params (AI%20Spawner%20Info%20Context :type AISpawner%20Info%20Context%20Structure))
  (return false 0))

(fn RequestSpawnAtTransformByDataComponentList
  (params (AISpawnerDataCompList :type Fort%20Athena%20AISpawner%20Data%20Component%20List%20Object%20Reference) (Transform :type Transform))
  (return false 0 0))
